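# Editor utility Blueprint that provides an action for viewing actor count details as part of the ActorCountVolume perf tools.
# Its View Actor Count Details function uses the EditorUtilitySubsystem to spawn and register a tab running the EUW_ActorCountVolumeStats editor widget.
# The event graph is empty and the asset holds no variables.
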EditorUtilityBlueprint /Game/NeverCook/EditorUtilities/PerfTools/ActorCountVolume/Data/ActorCountVolumeActions
graphs [
 EventGraph {links [], nodes []}
 View Actor Count Details {nodes [
   View Actor Count Details() {is=title K2Node_FunctionEntry}
   > SpawnAndRegisterTab(self=Get EditorUtilitySubsystem(), InBlueprint=/Game/NeverCook/EditorUtilities/PerfTools/ActorCountVolume/EUW_ActorCountVolumeStats)
  ]}
]
variables []
