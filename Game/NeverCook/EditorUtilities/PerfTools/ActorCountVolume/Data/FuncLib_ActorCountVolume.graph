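# Editor utility function library for the ActorCountVolume perf tool that tests actor descriptors against parent classes.
# IsDescInClassList loops over the classes and returns the first match along with the descriptor.
# IsDescOfClass checks the descriptor's native class first, avoiding a load when that matches.
# Otherwise it blocking-loads the class from its soft path and tests it for equality or inheritance, returning false if neither holds.
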EditorUtilityBlueprint /Game/NeverCook/EditorUtilities/PerfTools/ActorCountVolume/Data/FuncLib_ActorCountVolume
graphs [
 IsDescInClassList {nodes [
   n1 = IsDescInClassList!() -> (then, ActorDesc:Actor Desc Structure, ParentClasses:Array of Object Class References) {is=title K2Node_FunctionEntry}
   > n2 = For Each Loop(Array:Array of Object Class References=@n1.ParentClasses) -> (Array Element:Object Class Reference)
   > n3 = IsDescOfClass!(execute, __WorldContext:Object Reference, ActorDesc:Actor Desc Structure=@n1, ParentClass:Object Class Reference=@n2) -> (then, IsValid:bool)
   > Branch(Condition=@n3)
   > Return Node!(execute, IsInList:bool=true, Class:Object Class Reference=@n2, ActorDescOut:Actor Desc Structure=@n1)
  ]}
 IsDescOfClass {nodes [
   n1 = IsDescOfClass!() -> (then, ActorDesc:Actor Desc Structure, ParentClass:Object Class Reference) {is=title K2Node_FunctionEntry}
   > Set l_desc(l_desc:Actor Desc Structure=@n1) -> (Output_Get:Actor Desc Structure)
   > Set l_parentClass(l_parentClass:Object Class Reference=@n1.ParentClass) -> (Output_Get:Object Class Reference)
   > n2 = Sequence!(execute) -> (then_0, then_1, then_2:Exec)
   n3 = Break Actor Desc!(ActorDesc:"Actor Desc Structure (by ref)"=Get l_desc() -> (l_desc:Actor Desc Structure)) -> (NativeClass:Object Class Reference, Class:Soft Object Path Structure, ActorPackage:Name, FolderPath:Name) {is=title K2Node_BreakStruct}
   n4 = Get l_parentClass() -> (l_parentClass:Object Class Reference)
   > Branch(Condition=OR Boolean(A=ClassIsChildOf(TestClass=@n3, ParentClass=@n4), B="Equal (Class)"(A=@n3)))
   > Return Node!(execute, IsValid:bool=true)
   n5 = BreakSoftObjectPath(InSoftObjectPath=Break Actor Desc!(ActorDesc=Get l_desc()) -> (Class, FolderPath))
   n6 = LoadClassAsset_Blocking(execute=@n2.then_1, AssetClass=To Soft Class Reference(SoftClassPath=MakeSoftClassPath(PathString=@n5)))
   > n7 = Branch(Condition="Equal (Class)"(A=@n6, B=Get l_parentClass()))
   Return Node(execute=[@n8.then, @n7.then], IsValid=true)
   n8 = Branch(execute=@n7.else, Condition=ClassIsChildOf(TestClass=@n6, ParentClass=Get l_parentClass()))
   Return Node(execute=@n2.then_2, IsValid=false)
  ]}
]
variables []
